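# Physics asset for the M_MED_Sleek_Glasses_Pack backpack cosmetic, defining collision and joint limits for its dangling ring and two sticks.
# Small capsules on Dyn_ring and both Dyn_stick bones each constrain directly to the pack bone, whose zero-length capsule acts as a large sphere.
# The ring swings widely on one axis with the others nearly locked, while the sticks allow only slight twist and almost no swing.
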PhysicsAsset /BRCosmetics/Accessories/FORT_Backpacks/M_MED_Sleek_Glasses_Pack/Meshes/M_MED_Sleek_Glasses_Pack_Physics
bodies {
 Dyn_ring [{shapeType=Capsule, length=4, radius=1, shapeName=Dyn_ring_capsule, center=<0 0 -3>, rotation=<rot 90 90 0>}]
 Dyn_stick_01 [{shapeType=Capsule, length=10, radius=1, shapeName=Dyn_stick_01_capsule, center=<0 0 0.5>, rotation=<rot 0 0 89.9999>}]
 Dyn_stick_02 [{shapeType=Capsule, length=10, radius=1, shapeName=Dyn_stick_02_capsule, center=<0 0 0.7>, rotation=<rot 0 0 89.9999>}]
 pack [{shapeType=Capsule, length=0, radius=15.25, center=<0 -8.2007 0>, rotation=<rot 0 0 89.9999>}]
}
constraints [
 {bone1Name=Dyn_ring, bone2Name=pack, swing1LimitDegrees=5, swing2LimitDegrees=130, twistLimitDegrees=5}
 {bone1Name=Dyn_stick_01, bone2Name=pack, swing1LimitDegrees=1, swing2LimitDegrees=1, twistLimitDegrees=15}
 {bone1Name=Dyn_stick_02, bone2Name=pack, swing1LimitDegrees=1, swing2LimitDegrees=1, twistLimitDegrees=15}
]
