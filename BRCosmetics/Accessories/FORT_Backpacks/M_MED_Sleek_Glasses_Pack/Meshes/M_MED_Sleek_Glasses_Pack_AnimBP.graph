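# Animation Blueprint for the Sleek Glasses Pack backpack cosmetic, adding physics motion and repositioning the pack during skydiving.
# The reference pose gets a downward pack bone offset, then a RigidBody node simulates it in base-bone space with custom gravity and damped component motion, cached as 'Normal'.
# Each update combines skydiving, diving and floating flags into IsSkydiveAnim, which blends in an extra offset and roll.
# The final pose scales the pack bone down slightly.
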AnimBlueprint /BRCosmetics/Accessories/FORT_Backpacks/M_MED_Sleek_Glasses_Pack/Meshes/M_MED_Sleek_Glasses_Pack_AnimBP
target_skeleton=&/M_MED_Sleek_Glasses_Pack_Skeleton
graphs [
 AnimGraph {nodes [
   n1 = "Transform (Modify) Bone - Bone: pack"(ComponentPose=Local To Component(LocalPose=Use cached pose 'Normal'()), Translation="0,-4,5", Rotation="0,0,5") {
    rotationMode=BMM_Additive
    translationMode=BMM_Additive
    alphaBoolBlend {blendOption=Linear}
    boneToModify {boneName=pack}
    rotation {roll=5}
    translation {y=-4, z=5}
   }
   n2 = Component To Local(ComponentPose=@n1)
   n3 = Use cached pose 'Normal'()
   n4 = Get IsSkydiveAnim() -> (IsSkydiveAnim:bool)
   n5 = "Transform (Modify) Bone - Bone: pack"(
    ComponentPose=Local To Component(LocalPose=Blend Poses by bool(BlendPose_0=@n2, BlendPose_1=@n3, bActiveValue=@n4) {blendType=Linear})
    Translation="0,0,0"
    Rotation="0,0,0"
    Scale="0.9,0.9,0.9"
   ) {scaleMode=BMM_Replace, alphaBoolBlend {blendOption=Linear}, boneToModify {boneName=pack}, scale=<0.9 0.9 0.9>}
   Output Pose(Result=Component To Local(ComponentPose=@n5)) {layerGroup=None}
   n6 = "Transform (Modify) Bone - Bone: pack"(ComponentPose=Local To Component(LocalPose=Local Space Ref Pose()), Translation="0,0,-7", Rotation="0,0,0") {translationMode=BMM_Additive, alphaBoolBlend {blendOption=Linear}, boneToModify {boneName=pack}, translation {z=-7}}
   n7 = RigidBody(ComponentPose=@n6) {
    bOverrideWorldGravity=true
    overridePhysicsAsset=&/M_MED_Sleek_Glasses_Pack_Physics
    simulationSpace=BaseBoneSpace
    alphaBoolBlend {blendOption=Linear}
    baseBoneRef {boneName=pack}
    componentAppliedLinearAccClamp=<5000 5000 5000>
    componentLinearAccScale=<0.3 0.4 0.3>
    componentLinearVelScale=<0.3 0.4 0.3>
    overrideWorldGravity {z=-500}
   }
   Save cached pose 'Normal'(Pose=Component To Local(ComponentPose=@n7)) {cacheName=Normal}
   OR Boolean!(A=Get bIsSkydiveFloating(), B=Get bIsSkydiveDiveMode(), C:bool=Get bIsSkyDiving()) -> (ReturnValue)
  ]}
 EventGraph {nodes [
   Event BlueprintUpdateAnimation()
   > Set IsSkydiveAnim(IsSkydiveAnim:bool=OR Boolean(A=Get bIsSkydiveFloating(), B=Get bIsSkydiveDiveMode(), C=Get bIsSkyDiving())) -> (Output_Get:bool)
  ]}
]
variables [IsSkydiveAnim {type=boolean}]
